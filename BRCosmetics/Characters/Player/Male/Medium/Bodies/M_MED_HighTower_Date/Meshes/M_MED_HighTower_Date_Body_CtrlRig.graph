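# Control Rig for the M_MED_HighTower_Date player cosmetic body that corrects shoulder armor and front skirt bones.
# Additive shoulder offsets scale with arm angles and state variables for skydiving, glider, posing and celebrating, plus a front-end tweak.
# The front skirt follows a helper bone aimed between the knees, with a smoothed crouch offset.
# Spherical pose readers on both thighs sum into the SPR_thigh_l curve; several leftover skirt nodes are unconnected.
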ControlRigBlueprint /BRCosmetics/Characters/Player/Male/Medium/Bodies/M_MED_HighTower_Date/Meshes/M_MED_HighTower_Date_Body_CtrlRig
graphs [
 RigVMFunctionLibrary {links [], nodes []}
 RigVMModel {nodes [
   RigUnit_BeginExecution = &RigUnit_BeginExecution()
   SequenceAggregate = RigVMAggregateNode(ExecuteContext=@RigUnit_BeginExecution.ExecutePin)
   /Script/ControlRig.RigUnit_SetBoneTranslation(
    Bone=dyn_kneeAim_l
    Space=GlobalSpace
    Weight=1
    bPropagateToChildren=True
    Translation=&RigUnit_GetBoneTransform(Bone=calf_l).Transform.Translation
    ExecutePin=@SequenceAggregate.A
   ) {is=struct RigVMUnitNode}
   > /Script/ControlRig.RigUnit_AimBone(
    Bone=dyn_kneeAim_l
    DebugSettings=(bEnabled=true,Scale=10,WorldOffset=(Rotation=(X=0,Y=0,Z=0,W=1),Translation=(X=0,Y=0,Z=0),Scale3D=(X=1,Y=1,Z=1)))
    Primary=(Weight=1,Axis=(X=0,Y=0,Z=1),Target=(X=1,Y=0,Z=0),Kind=Location)
    Secondary=(Weight=1,Axis=(X=10,Y=0.1,Z=0),Target=(X=1,Y=0,Z=0),Kind=Location)
    Weight=1
    bPropagateToChildren=True
    Primary.Target=&RigUnit_GetBoneTransform(Bone=calf_r).Transform.Translation
    Secondary.Target=&RigUnit_GetBoneTransform(Bone=pelvis).Transform.Translation
   ) {is=struct RigVMUnitNode}
   n1 = /Script/RigVM.RigVMFunction_MathFloatAbs(Value=&RigVMFunction_MathQuaternionToEuler(RotationOrder=ZYX, Value=&RigUnit_GetRelativeBoneTransform(Bone=upperarm_r, Space=spine_05).Transform.Rotation).Result.Z) {is=struct RigVMUnitNode, template="Absolute::Execute(in Value,out Result)"}
   n2 = RigVMVariableNode() {variable=isSkydiving}
   > /Script/ControlRig.RigUnit_ModifyBoneTransforms(
    BoneToModify=((Bone="dyn_shoulderScale_r",Transform=(Rotation=(X=0,Y=0,Z=0,W=1),Translation=(X=-2.5,Y=-3.75,Z=-1),Scale3D=(X=1,Y=1,Z=1))))
    Mode=AdditiveLocal
    WeightMaximum=1
    Weight=&RigVMFunction_MathFloatMul(
     A=&RigVMFunction_MathFloatRemap(SourceMaximum=123, SourceMinimum=15, TargetMinimum=0, bClamp=true, Value=@n1.Result).Result
     B=@n2.Value
    ) {template="Multiply::Execute(in A,in B,out Result)"}.Result
    BoneToModify.0.Transform.Rotation=&RigVMFunction_MathQuaternionFromEuler(Euler=(X=-90,Y=26,Z=46)).Result
   ) {is=struct RigVMUnitNode}
   > &RigUnit_ModifyBoneTransforms(
    BoneToModify=((Bone="dyn_shoulderScale_r",Transform=(Rotation=(X=0,Y=0,Z=0,W=1),Translation=(X=3,Y=-2,Z=3),Scale3D=(X=1,Y=1,Z=1))))
    Mode=AdditiveLocal
    WeightMaximum=1
    Weight=&RigVMFunction_MathFloatMul(
     A=&RigVMFunction_MathFloatRemap(SourceMaximum=142, SourceMinimum=124, TargetMinimum=0, bClamp=true, Value=@n1.Result).Result
     B=@n2.Value
    ) {template="Multiply::Execute(in A,in B,out Result)"}.Result
    BoneToModify.0.Transform.Rotation=&RigVMFunction_MathQuaternionFromEuler(Euler=(X=20,Y=-20,Z=-25)).Result
   )
   n3 = &RigVMFunction_MathFloatAbs(Value=&RigVMFunction_MathQuaternionToEuler(RotationOrder=ZYX, Value=&RigUnit_GetRelativeBoneTransform(Bone=upperarm_l, Space=spine_05).Transform.Rotation).Result.Z) {template="Absolute::Execute(in Value,out Result)"}
   > &RigUnit_ModifyBoneTransforms(
    BoneToModify=((Bone="dyn_shoulderScale_l",Transform=(Rotation=(X=0,Y=0,Z=0,W=1),Translation=(X=-0.5,Y=1,Z=1),Scale3D=(X=1,Y=1,Z=1))))
    Mode=AdditiveLocal
    WeightMaximum=1
    Weight=&RigVMFunction_MathFloatMul(
     A=&RigVMFunction_MathFloatRemap(SourceMaximum=60, SourceMinimum=120, TargetMinimum=0, bClamp=true, Value=@n3.Result).Result
     B=@n2.Value
    ) {template="Multiply::Execute(in A,in B,out Result)"}.Result
    BoneToModify.0.Transform.Rotation=&RigVMFunction_MathQuaternionFromEuler(Euler=(X=-75,Y=20,Z=25)).Result
   )
   n4 = RigVMVariableNode() {variable=isCrouching}
   > &RigUnit_ModifyBoneTransforms(
    BoneToModify=((Bone="dyn_skirtFrontMid_Base",Transform=(Rotation=(X=0,Y=0,Z=0,W=1),Translation=(X=-12,Y=5,Z=0),Scale3D=(X=1,Y=1,Z=1))))
    Mode=AdditiveLocal
    WeightMaximum=1
    Weight=/Script/RigVM.RigVMFunction_AccumulateFloatLerp(Blend=8, bIntegrateDeltaTime=true, InitialValue=@n4.Value, TargetValue=@n4.Value) {is=struct RigVMUnitNode, template="AccumulateLerp::Execute(in TargetValue,in InitialValue,in Blend,in bIntegrateDeltaTime,out Result)"}.Result
   )
   > /Script/ControlRig.RigUnit_TransformConstraint(
    BaseBone=pelvis
    BaseTransform=(Rotation=(X=0,Y=0,Z=0,W=1),Translation=(X=0,Y=0,Z=0),Scale3D=(X=1,Y=1,Z=1))
    BaseTransformSpace=GlobalSpace
    Bone=dyn_skirtFrontMid_Base
    Targets=((Transform=(Rotation=(X=0,Y=0,Z=0,W=1),Translation=(X=0,Y=0,Z=0),Scale3D=(X=1,Y=1,Z=1)),Weight=1,bMaintainOffset=True,Filter=(TranslationFilter=(bX=True,bY=True,bZ=True),RotationFilter=(bX=True,bY=True,bZ=True),ScaleFilter=(bX=True,bY=True,bZ=True))))
    bUseInitialTransforms=True
    Targets.0.Transform.Rotation=&RigUnit_GetBoneTransform(Bone=dyn_kneeAim_l).Transform.Rotation
    Targets.0.Transform.Translation=&RigUnit_GetBoneTransform(Bone=dyn_skirtFrontMid_Base).Transform.Translation
   ) {is=struct RigVMUnitNode}
   n5 = &RigVMFunction_MathFloatMul(
    A=/Script/RigVM.RigVMFunction_MathFloatSub(A=1, B=RigVMVariableNode() {variable=isSkydiving}.Value) {is=struct RigVMUnitNode, template="Subtract::Execute(in A,in B,out Result)"}.Result
    B=&RigVMFunction_MathFloatSub(A=1, B=RigVMVariableNode() {variable=dateGliderDeployed}.Value) {template="Subtract::Execute(in A,in B,out Result)"}.Result
   ) {template="Multiply::Execute(in A,in B,out Result)"}
   GetRelativeBoneTransform_2_2 = &RigUnit_GetRelativeTransformForItem(Child=(Type=Bone,Name="upperarm_r"), Parent=(Type=Bone,Name="spine_05"), bChildInitial=False, bParentInitial=False)
   n6 = &RigVMFunction_MathFloatAbs(Value=&RigVMFunction_MathQuaternionToEuler(RotationOrder=ZYX, Value=@GetRelativeBoneTransform_2_2.RelativeTransform.Rotation).Result.Z) {template="Absolute::Execute(in Value,out Result)"}
   > &RigUnit_ModifyBoneTransforms(
    BoneToModify=((Bone="dyn_shoulderScale_r",Transform=(Rotation=(X=0,Y=0,Z=0,W=1),Translation=(X=2,Y=-3,Z=-3),Scale3D=(X=1,Y=1,Z=1))))
    Mode=AdditiveLocal
    WeightMaximum=1
    Weight=&RigVMFunction_MathFloatMul(
     B=@n5.Result
     A=&RigVMFunction_MathFloatRemap(SourceMaximum=120, SourceMinimum=30, TargetMinimum=0, bClamp=true, Value=@n6.Result).Result
    ) {template="Multiply::Execute(in A,in B,out Result)"}.Result
    BoneToModify.0.Transform.Rotation=&RigVMFunction_MathQuaternionFromEuler(Euler=(X=0,Y=0,Z=-40)).Result
   )
   > &RigUnit_ModifyBoneTransforms(
    BoneToModify=((Bone="dyn_shoulderArmor_Base_r",Transform=(Rotation=(X=0,Y=0,Z=0,W=1),Translation=(X=0,Y=5,Z=0),Scale3D=(X=1,Y=1,Z=1))),(Bone="dyn_shoulderScale_r",Transform=(Rotation=(X=0,Y=0,Z=0,W=1),Translation=(X=-2,Y=3,Z=3.5),Scale3D=(X=1,Y=1,Z=1))))
    Mode=AdditiveLocal
    WeightMaximum=1
    Weight=&RigVMFunction_MathFloatMul(
     B=@n5.Result
     A=&RigVMFunction_MathFloatRemap(SourceMaximum=148, SourceMinimum=120, TargetMinimum=0, bClamp=true, Value=@n6.Result).Result
    ) {template="Multiply::Execute(in A,in B,out Result)"}.Result
    BoneToModify.0.Transform.Rotation=&RigVMFunction_MathQuaternionFromEuler(Euler=(X=0,Y=0,Z=-64)).Result
    BoneToModify.1.Transform.Rotation=&RigVMFunction_MathQuaternionFromEuler(Euler=(X=0,Y=0,Z=40)).Result
   )
   GetRelativeBoneTransform_1_1_1 = &RigUnit_GetRelativeTransformForItem(Child=(Type=Bone,Name="upperarm_l"), Parent=(Type=Bone,Name="spine_05"), bChildInitial=False, bParentInitial=False)
   n7 = &RigVMFunction_MathFloatAbs(Value=&RigVMFunction_MathQuaternionToEuler(RotationOrder=ZYX, Value=@GetRelativeBoneTransform_1_1_1.RelativeTransform.Rotation).Result.Z) {template="Absolute::Execute(in Value,out Result)"}
   > &RigUnit_ModifyBoneTransforms(
    BoneToModify=((Bone="dyn_shoulderScale_l",Transform=(Rotation=(X=0,Y=0,Z=0,W=1),Translation=(X=-4,Y=4,Z=2),Scale3D=(X=1,Y=1,Z=1))),(Bone="dyn_shoulderArmor_Base_l",Transform=(Rotation=(X=0,Y=0,Z=0,W=1),Translation=(X=0,Y=0,Z=0),Scale3D=(X=1,Y=1,Z=1))))
    Mode=AdditiveLocal
    WeightMaximum=1
    Weight=&RigVMFunction_MathFloatMul(
     B=@n5.Result
     A=&RigVMFunction_MathFloatRemap(SourceMaximum=70, SourceMinimum=150, TargetMinimum=0, bClamp=true, Value=@n7.Result).Result
    ) {template="Multiply::Execute(in A,in B,out Result)"}.Result
    BoneToModify.0.Transform.Rotation=&RigVMFunction_MathQuaternionFromEuler(Euler=(X=0,Y=0,Z=-40)).Result
    BoneToModify.1.Transform.Rotation=&RigVMFunction_MathQuaternionFromEuler(Euler=(X=0,Y=0,Z=10)).Result
   )
   > &RigUnit_ModifyBoneTransforms(
    BoneToModify=((Bone="dyn_shoulderScale_l",Transform=(Rotation=(X=0,Y=0,Z=0,W=1),Translation=(X=-3,Y=3,Z=-1),Scale3D=(X=1,Y=1,Z=1))))
    Mode=AdditiveLocal
    WeightMaximum=1
    Weight=&RigVMFunction_MathFloatMul(
     B=@n5.Result
     A=&RigVMFunction_MathFloatRemap(SourceMaximum=20, SourceMinimum=70, TargetMinimum=0, bClamp=true, Value=@n7.Result).Result
    ) {template="Multiply::Execute(in A,in B,out Result)"}.Result
    BoneToModify.0.Transform.Rotation=&RigVMFunction_MathQuaternionFromEuler(Euler=(X=0,Y=0,Z=-48)).Result
   )
   n8 = &RigVMFunction_MathFloatAbs(Value=&RigVMFunction_MathQuaternionToEuler(RotationOrder=ZYX, Value=&RigUnit_GetRelativeBoneTransform(Bone=upperarm_r, Space=spine_05).Transform.Rotation).Result.Z) {template="Absolute::Execute(in Value,out Result)"}
   n9 = RigVMVariableNode() {variable=dateGliderDeployed}
   > &RigUnit_ModifyBoneTransforms(
    BoneToModify=((Bone="dyn_shoulderScale_r",Transform=(Rotation=(X=0,Y=0,Z=0,W=1),Translation=(X=1.5,Y=-2,Z=-2),Scale3D=(X=1,Y=1,Z=1))))
    Mode=AdditiveLocal
    WeightMaximum=1
    Weight=&RigVMFunction_MathFloatMul(
     A=&RigVMFunction_MathFloatRemap(SourceMaximum=120, SourceMinimum=90, TargetMinimum=0, bClamp=true, Value=@n8.Result).Result
     B=@n9.Value
    ) {template="Multiply::Execute(in A,in B,out Result)"}.Result
    BoneToModify.0.Transform.Rotation=&RigVMFunction_MathQuaternionFromEuler(Euler=(X=-30,Y=38,Z=-15)).Result
   )
   > &RigUnit_ModifyBoneTransforms(
    BoneToModify=((Bone="dyn_shoulderArmor_Base_r",Transform=(Rotation=(X=0,Y=0,Z=0,W=1),Translation=(X=0,Y=0,Z=0),Scale3D=(X=1,Y=1,Z=1))),(Bone="dyn_shoulderScale_r",Transform=(Rotation=(X=0,Y=0,Z=0,W=1),Translation=(X=0,Y=0,Z=0),Scale3D=(X=1,Y=1,Z=1))))
    Mode=AdditiveLocal
    WeightMaximum=1
    Weight=&RigVMFunction_MathFloatMul(
     A=&RigVMFunction_MathFloatRemap(SourceMaximum=160, SourceMinimum=120, TargetMinimum=0, bClamp=true, Value=@n8.Result).Result
     B=@n9.Value
    ) {template="Multiply::Execute(in A,in B,out Result)"}.Result
    BoneToModify.1.Transform.Rotation=&RigVMFunction_MathQuaternionFromEuler(Euler=(X=0,Y=0,Z=0)).Result
    BoneToModify.0.Transform.Rotation=&RigVMFunction_MathQuaternionFromEuler(Euler=(X=0,Y=0,Z=0)).Result
   )
   n10 = &RigVMFunction_MathFloatAbs(Value=&RigVMFunction_MathQuaternionToEuler(RotationOrder=ZYX, Value=&RigUnit_GetRelativeBoneTransform(Bone=upperarm_l, Space=spine_05).Transform.Rotation).Result.Z) {template="Absolute::Execute(in Value,out Result)"}
   > &RigUnit_ModifyBoneTransforms(
    BoneToModify=((Bone="dyn_shoulderScale_l",Transform=(Rotation=(X=0,Y=0,Z=0,W=1),Translation=(X=-4,Y=4,Z=2),Scale3D=(X=1,Y=1,Z=1))),(Bone="dyn_shoulderArmor_Base_l",Transform=(Rotation=(X=0,Y=0,Z=0,W=1),Translation=(X=0,Y=0,Z=0),Scale3D=(X=1,Y=1,Z=1))))
    Mode=AdditiveLocal
    WeightMaximum=1
    Weight=&RigVMFunction_MathFloatMul(
     A=&RigVMFunction_MathFloatRemap(SourceMaximum=70, SourceMinimum=150, TargetMinimum=0, bClamp=true, Value=@n10.Result).Result
     B=@n9.Value
    ) {template="Multiply::Execute(in A,in B,out Result)"}.Result
    BoneToModify.1.Transform.Rotation=&RigVMFunction_MathQuaternionFromEuler(Euler=(X=0,Y=0,Z=10)).Result
    BoneToModify.0.Transform.Rotation=&RigVMFunction_MathQuaternionFromEuler(Euler=(X=0,Y=0,Z=-40)).Result
   )
   > &RigUnit_ModifyBoneTransforms(
    BoneToModify=((Bone="dyn_shoulderScale_l",Transform=(Rotation=(X=0,Y=0,Z=0,W=1),Translation=(X=-3,Y=3,Z=-1),Scale3D=(X=1,Y=1,Z=1))))
    Mode=AdditiveLocal
    WeightMaximum=1
    Weight=&RigVMFunction_MathFloatMul(
     A=&RigVMFunction_MathFloatRemap(SourceMaximum=20, SourceMinimum=70, TargetMinimum=0, bClamp=true, Value=@n10.Result).Result
     B=@n9.Value
    ) {template="Multiply::Execute(in A,in B,out Result)"}.Result
    BoneToModify.0.Transform.Rotation=&RigVMFunction_MathQuaternionFromEuler(Euler=(X=0,Y=0,Z=-48)).Result
   )
   n11 = &RigVMFunction_MathFloatAbs(Value=&RigVMFunction_MathQuaternionToEuler(RotationOrder=ZYX, Value=&RigUnit_GetRelativeBoneTransform(Bone=upperarm_r, Space=spine_05).Transform.Rotation).Result.Z) {template="Absolute::Execute(in Value,out Result)"}
   n12 = &RigVMFunction_MathFloatMul(
    B=1
    A=&RigVMFunction_MathFloatRemap(SourceMaximum=14.8, SourceMinimum=6, TargetMinimum=0.9, bClamp=true, Value=@n11.Result).Result
   ) {template="Multiply::Execute(in A,in B,out Result)"}
   n13 = RigVMVariableNode() {variable=isPosing}
   n14 = &RigVMFunction_MathQuaternionToEuler(RotationOrder=ZYX, Value=&RigUnit_GetRelativeBoneTransform(Bone=lowerarm_r, Space=upperarm_r).Transform.Rotation)
   > &RigUnit_ModifyBoneTransforms(
    BoneToModify=((Bone="dyn_shoulderScale_r",Transform=(Rotation=(X=0,Y=0,Z=0,W=1),Translation=(X=-1,Y=-1.5,Z=0),Scale3D=(X=1,Y=1,Z=1))),(Bone="dyn_shoulderArmor_Base_r",Transform=(Rotation=(X=0,Y=0,Z=0,W=1),Translation=(X=0,Y=0.8,Z=0),Scale3D=(X=1,Y=1,Z=1))))
    Mode=AdditiveLocal
    WeightMaximum=1
    Weight=&RigVMFunction_MathFloatMul(
     A=&RigVMFunction_MathFloatMul(A=@n12.Result, B=@n13.Value) {template="Multiply::Execute(in A,in B,out Result)"}.Result
     B=&RigVMFunction_MathFloatRemap(SourceMaximum=101, SourceMinimum=35, TargetMinimum=0, bClamp=true, Value=@n14.Result.Z).Result
    ) {template="Multiply::Execute(in A,in B,out Result)"}.Result
    BoneToModify.0.Transform.Rotation=&RigVMFunction_MathQuaternionFromEuler(Euler=(X=-20,Y=15,Z=12)).Result
    BoneToModify.1.Transform.Rotation=&RigVMFunction_MathQuaternionFromEuler(Euler=(X=-3,Y=0,Z=0)).Result
   )
   n15 = &RigUnit_GetRelativeBoneTransform(Bone=upperarm_l, Space=spine_05)
   n16 = &RigVMFunction_MathFloatRemap(
    SourceMaximum=168
    SourceMinimum=167
    TargetMinimum=0.8
    bClamp=true
    Value=&RigVMFunction_MathFloatAbs(Value=&RigVMFunction_MathQuaternionToEuler(RotationOrder=ZYX, Value=@n15.Transform.Rotation).Result.Z) {template="Absolute::Execute(in Value,out Result)"}.Result
   )
   n17 = &RigVMFunction_MathQuaternionToEuler(RotationOrder=ZYX, Value=&RigUnit_GetRelativeBoneTransform(Bone=lowerarm_l, Space=upperarm_l).Transform.Rotation)
   > &RigUnit_ModifyBoneTransforms(
    BoneToModify=((Bone="dyn_shoulderScale_l",Transform=(Rotation=(X=0,Y=0,Z=0,W=1),Translation=(X=0,Y=0.8,Z=0),Scale3D=(X=1,Y=1,Z=1))),(Bone="dyn_shoulderArmor_Base_l",Transform=(Rotation=(X=0,Y=0,Z=0,W=1),Translation=(X=0,Y=-0.7,Z=1),Scale3D=(X=1,Y=1,Z=1))))
    Mode=AdditiveLocal
    WeightMaximum=1
    Weight=&RigVMFunction_MathFloatMul(
     B=&RigVMFunction_MathFloatMul(A=@n16.Result, B=@n13.Value) {template="Multiply::Execute(in A,in B,out Result)"}.Result
     A=&RigVMFunction_MathFloatRemap(SourceMaximum=100, SourceMinimum=35, TargetMinimum=0, bClamp=true, Value=@n17.Result.Z).Result
    ) {template="Multiply::Execute(in A,in B,out Result)"}.Result
    BoneToModify.1.Transform.Rotation=&RigVMFunction_MathQuaternionFromEuler(Euler=(X=0,Y=0,Z=0)).Result
    BoneToModify.0.Transform.Rotation=&RigVMFunction_MathQuaternionFromEuler(Euler=(X=-15,Y=10,Z=0)).Result
   )
   n18 = RigVMVariableNode() {variable=isCelebrating}
   > &RigUnit_TransformConstraint(
    BaseTransform=(Rotation=(X=0,Y=0,Z=0,W=1),Translation=(X=0,Y=0,Z=0),Scale3D=(X=1,Y=1,Z=1))
    BaseTransformSpace=GlobalSpace
    Bone=dyn_shoulderArmor_Base_r
    Targets="((Transform=(Rotation=(X=0,Y=0,Z=0,W=1),Translation=(X=0,Y=0,Z=0),Scale3D=(X=1,Y=1,Z=1)),Weight=1,bMaintainOffset=True,Filter=(TranslationFilter=(bX=True,bY=True,bZ=True),RotationFilter=(bX=True,bY=True,bZ=True),ScaleFilter=(bX=True,bY=True,bZ=True))),(Tra ...(truncated)"
    bUseInitialTransforms=True
    Targets.0.Transform=&RigUnit_GetBoneTransform(Bone=upperarm_twist_01_r).Transform
    Targets.0.Weight=@n18.Value
   )
   > &RigUnit_TransformConstraint(
    BaseTransform=(Rotation=(X=0,Y=0,Z=0,W=1),Translation=(X=0,Y=0,Z=0),Scale3D=(X=1,Y=1,Z=1))
    BaseTransformSpace=GlobalSpace
    Bone=dyn_shoulderArmor_Base_l
    Targets="((Transform=(Rotation=(X=0,Y=0,Z=0,W=1),Translation=(X=0,Y=0,Z=0),Scale3D=(X=1,Y=1,Z=1)),Weight=1,bMaintainOffset=True,Filter=(TranslationFilter=(bX=True,bY=True,bZ=True),RotationFilter=(bX=True,bY=True,bZ=True),ScaleFilter=(bX=True,bY=True,bZ=True))),(Tra ...(truncated)"
    bUseInitialTransforms=True
    Targets.0.Transform=&RigUnit_GetBoneTransform(Bone=upperarm_twist_01_l).Transform
    Targets.0.Weight=@n18.Value
   )
   > n19 = &RigUnit_ModifyBoneTransforms(
    BoneToModify=((Bone="dyn_shoulderScale_r",Transform=(Rotation=(X=0,Y=0,Z=0,W=1),Translation=(X=0,Y=-2,Z=3),Scale3D=(X=1,Y=1,Z=1))),(Bone="dyn_shoulderScale_l",Transform=(Rotation=(X=0,Y=0,Z=0,W=1),Translation=(X=2,Y=2,Z=-2),Scale3D=(X=1,Y=1,Z=1))))
    Mode=AdditiveLocal
    WeightMaximum=1
    BoneToModify.1.Transform.Rotation=&RigVMFunction_MathQuaternionFromEuler(Euler=(X=0,Y=-30,Z=0)).Result
    BoneToModify.0.Transform.Rotation=&RigVMFunction_MathQuaternionFromEuler(Euler=(X=0,Y=-45,Z=-20)).Result
    Weight=@n18.Value
   )
   RigVMFunction_ControlFlowBranch = &RigVMFunction_ControlFlowBranch(ExecuteContext=@n19.ExecutePin, Condition=RigVMVariableNode() {variable=isFrontEnd}.Value)
   &RigUnit_ModifyTransforms(
    ItemToModify=((Item=(Type=Bone,Name="dyn_shoulderScale_l"),Transform=(Rotation=(X=0,Y=0,Z=0.0872,W=0.9962),Translation=(X=1,Y=0,Z=0),Scale3D=(X=1,Y=1,Z=1))))
    Weight=1
    WeightMinimum=0
    ExecutePin=@RigVMFunction_ControlFlowBranch.True
   )
   n20 = &RigUnit_SphericalPoseReader(
    ActiveRegionScaleFactors=(PositiveWidth=3.5,NegativeWidth=0.5,PositiveHeight=3,NegativeHeight=3)
    ActiveRegionSize=0.1
    Debug=(bDrawDebug=true,DebugScale=25,DebugSegments=20,DebugThickness=0.25)
    DriverAxis=(X=-1,Y=0,Z=0)
    DriverItem=(Type=Bone,Name="thigh_l")
    FalloffRegionScaleFactors=(PositiveWidth=1,NegativeWidth=1,PositiveHeight=1,NegativeHeight=1)
    FalloffSize=0.45
    OptionalParentItem=(Type=Bone)
    RotationOffset=(X=90,Y=0,Z=0)
    ExecutePin=@SequenceAggregate.B
   )
   n21 = &RigUnit_SphericalPoseReader(
    ActiveRegionScaleFactors=(PositiveWidth=3.5,NegativeWidth=0.5,PositiveHeight=3,NegativeHeight=3)
    ActiveRegionSize=0.1
    Debug=(bDrawDebug=true,DebugScale=25,DebugSegments=20,DebugThickness=0.25)
    DriverAxis=(X=1,Y=0,Z=0)
    DriverItem=(Type=Bone,Name="thigh_r")
    FalloffRegionScaleFactors=(PositiveWidth=1,NegativeWidth=1,PositiveHeight=1,NegativeHeight=1)
    FalloffSize=0.45
    FlipHeightScaling=false
    FlipWidthScaling=true
    OptionalParentItem=(Type=Bone)
    RotationOffset=(X=-90,Y=0,Z=0)
    ExecutePin=@SequenceAggregate.C
   )
   Add = &RigVMFunction_MathFloatAdd(A=@n20.OutputParam, B=@n21.OutputParam) {template="Add::Execute(in A,in B,out Result)"}
   Clamp = /Script/RigVM.RigVMFunction_MathFloatClamp(Maximum=1, Minimum=0, Value=@Add.Result) {is=struct RigVMUnitNode, template="Clamp::Execute(in Value,in Minimum,in Maximum,out Result)"}
   > /Script/ControlRig.RigUnit_SetCurveValue(Curve=SPR_thigh_l, Value=@Clamp.Result) {is=struct RigVMUnitNode}
   &RigUnit_TransformConstraint(
    BaseTransform=(Rotation=(X=0,Y=0,Z=0,W=1),Translation=(X=0,Y=0,Z=0),Scale3D=(X=1,Y=1,Z=1))
    BaseTransformSpace=GlobalSpace
    Bone=dyn_skirtFrontMid
    ExecutePin=()
    Targets="((Transform=(Rotation=(X=0,Y=0,Z=0,W=1),Translation=(X=0,Y=0,Z=0),Scale3D=(X=1,Y=1,Z=1)),Weight=1,bMaintainOffset=false,Filter=(TranslationFilter=(bX=True,bY=True,bZ=True),RotationFilter=(bX=True,bY=True,bZ=True),ScaleFilter=(bX=True,bY=True,bZ=True))),(Tr ...(truncated)"
    bUseInitialTransforms=True
   )
   > &RigUnit_SetBoneTranslation(Bone=dyn_frontLegSplit, Space=LocalSpace, Translation=(X=0,Y=0,Z=0), Weight=1, bPropagateToChildren=True)
   &RigUnit_AimBone(
    Bone=dyn_skirtFrontMid
    DebugSettings=(Scale=10,WorldOffset=(Rotation=(X=0,Y=0,Z=0,W=1),Translation=(X=0,Y=0,Z=0),Scale3D=(X=1,Y=1,Z=1)))
    ExecutePin=()
    Primary=(Weight=1,Axis=(X=1,Y=0,Z=0),Target=(X=1,Y=0,Z=0),Kind=Location)
    Secondary=(Weight=1,Axis=(X=0,Y=1,Z=0),Target=(X=1,Y=0,Z=0),Kind=Location)
    Weight=1
    bPropagateToChildren=True
    Primary.Target=&RigUnit_GetBoneTransform(Bone=dyn_frontLegSplit).Transform.Translation
    Secondary.Target=&RigUnit_GetBoneTransform(Bone=pelvis).Transform.Translation
   )
   &RigUnit_GetBoneTransform(Bone=thigh_twist_01_l)
   &RigUnit_GetBoneTransform(Bone=thigh_twist_01_r)
   &RigVMFunction_MathFloatAbs(Value=&RigVMFunction_MathQuaternionToEuler(RotationOrder=ZYX, Value=&RigUnit_GetRelativeBoneTransform(Bone=upperarm_r, Space=clavicle_r).Transform.Rotation).Result.Z) {template="Absolute::Execute(in Value,out Result)"}
   n22 = &RigUnit_GetRelativeBoneTransform(Bone=thigh_r, Space=pelvis)
   n23 = &RigVMFunction_MathFloatRemap(
    SourceMaximum=80
    SourceMinimum=148
    TargetMaximum=12
    TargetMinimum=0
    bClamp=true
    Value=&RigVMFunction_MathFloatAbs(Value=&RigVMFunction_MathQuaternionToEuler(RotationOrder=ZYX, Value=@n22.Transform.Rotation).Result.Z) {template="Absolute::Execute(in Value,out Result)"}.Result
   )
   n24 = &RigUnit_GetRelativeBoneTransform(Bone=thigh_l, Space=pelvis)
   n25 = &RigVMFunction_MathFloatRemap(
    SourceMaximum=110
    SourceMinimum=30
    TargetMaximum=12
    TargetMinimum=0
    bClamp=true
    Value=&RigVMFunction_MathFloatAbs(Value=&RigVMFunction_MathQuaternionToEuler(RotationOrder=ZYX, Value=@n24.Transform.Rotation).Result.Z) {template="Absolute::Execute(in Value,out Result)"}.Result
   )
   n26 = &RigVMFunction_MathFloatAdd(
    B=15
    A=/Script/RigVM.RigVMFunction_MathFloatDiv(B=2, A=&RigVMFunction_MathFloatAdd(B=@n23.Result, A=@n25.Result) {template="Add::Execute(in A,in B,out Result)"}.Result) {is=struct RigVMUnitNode, template="Divide::Execute(in A,in B,out Result)"}.Result
   ) {template="Add::Execute(in A,in B,out Result)"}
   &RigVMFunction_MathFloatRemap(SourceMaximum=24, SourceMinimum=0, TargetMinimum=0, bClamp=true, Value=@n26.Result)
   &RigVMFunction_MathFloatRemap(SourceMaximum=26, SourceMinimum=14.8, TargetMaximum=0, TargetMinimum=1, bClamp=true, Value=@n11.Result)
   &RigVMFunction_MathFloatSub(A=1, B=@n2.Value) {template="Subtract::Execute(in A,in B,out Result)"}
   &RigVMFunction_MathQuaternionFromEuler(Euler=(X=0,Y=0,Z=0))
   /Script/RigVM.RigVMFunction_MathVectorAdd(B=(X=-20,Y=10,Z=0), A=&RigUnit_GetBoneTransform(Bone=dyn_frontLegSplit).Transform.Translation, B.Y=@n26.Result) {is=struct RigVMUnitNode, template="Add::Execute(in A,in B,out Result)"}
   /Script/RigVM.RigVMFunction_MathVectorCross(A=(X=0,Y=0,Z=0), B=(X=0,Y=0,Z=0)) {is=struct RigVMUnitNode, template="Cross::Execute(in A,in B,out Result)"}
   &RigUnit_ModifyBoneTransforms(
    BoneToModify=((Bone="dyn_skirtFrontMid",Transform=(Rotation=(X=0,Y=0,Z=0,W=1),Translation=(X=0,Y=0,Z=0),Scale3D=(X=1,Y=1,Z=1))),(Transform=(Rotation=(X=0,Y=0,Z=0,W=1),Translation=(X=0,Y=0,Z=0),Scale3D=(X=1,Y=1,Z=1))))
    ExecutePin=()
    Mode=AdditiveLocal
    Weight=1
    WeightMaximum=1
   )
  ]}
 SequenceAggregate_1_ContainedGraph {nodes [
   n1 = RigVMFunctionEntryNode()
   SequenceAggregate = &RigVMFunction_Sequence(ExecuteContext=@n1.ExecuteContext)
   SequenceAggregate_1 = &RigVMFunction_Sequence(ExecuteContext=@SequenceAggregate.B)
   RigVMFunctionReturnNode(A=@SequenceAggregate.A, B=@SequenceAggregate_1.A, C=@SequenceAggregate_1.B)
  ]}
]
hierarchy [
 root {loc=<0 0 0>, rot=<0 0 0 1>}
 pelvis {parent=root, loc=<0 -0.7071 93.1191>, rot=<0.0127 0.707 -0.0127 -0.707>}
 spine_01 {parent=pelvis, loc=<6.7494 -0.6388 0>, rot=<0 0 0.034 -0.9994>}
 spine_02 {parent=spine_01, loc=<7.2518 -0.035 0>, rot=<0 0 0.0238 -0.9997>}
 spine_03 {parent=spine_02, loc=<6.0104 0 0>, rot=<0 0 0.0149 -0.9999>}
 spine_04 {parent=spine_03, loc=<8.1024 0 0>, rot=<0 0 -0.0132 -0.9999>}
 spine_05 {parent=spine_04, loc=<10.0146 0 0>, rot=<0 0 -0.0015 -1>}
 clavicle_l {parent=spine_05, loc=<4.3575 -0.1466 -2.3223>, rot=<0.0465 -0.7289 -0.009 -0.6829>}
 upperarm_l {parent=clavicle_l, loc=<15.5111 0 0>, rot=<0.0226 -0.3234 -0.0088 -0.9459>}
 upperarm_twist_01_l {parent=upperarm_l, loc=<0.0041 0 0>, rot=<0 0 0 -1>}
 lowerarm_l {parent=upperarm_l, loc=<26.6514 0 0>, rot=<0.0391 -0.0099 -0.2453 -0.9686>}
 hand_l {parent=lowerarm_l, loc=<22.3491 0 0>, rot=<-0.7081 -0.0507 0.0975 -0.6975>}
 index_metacarpal_l {parent=hand_l, loc=<1.6535 0.0183 -2.96>, rot=<0.0158 -0.0124 -0.0173 -0.9996>}
 index_01_l {parent=index_metacarpal_l, loc=<8.3008 0.0357 -0.0524>, rot=<-0.0044 0.0485 0.0924 -0.9945>}
 index_02_l {parent=index_01_l, loc=<3.3587 -0.3004 0.0452>, rot=<-0.002 0.0027 0.3225 -0.9466>}
 index_03_l {parent=index_02_l, loc=<2.5578 -0.123 -0.0252>, rot=<-0.002 0.0067 0.2558 -0.9667>}
 middle_metacarpal_l {parent=hand_l, loc=<2.1856 -0.1725 -0.5984>, rot=<0.006 0.0168 -0.0329 -0.9993>}
 middle_01_l {parent=middle_metacarpal_l, loc=<7.3275 -0.2447 0.0129>, rot=<-0.0208 0.0042 0.0858 -0.9961>}
 middle_02_l {parent=middle_01_l, loc=<4.0123 0.1401 0.1175>, rot=<-0.0218 0.0258 0.3512 -0.9357>}
 middle_03_l {parent=middle_02_l, loc=<2.6077 0.0534 0.1373>, rot=<0.0082 0.0188 0.1708 -0.9851>}
 pinky_metacarpal_l {parent=hand_l, loc=<1.7616 -0.9753 3.3814>, rot=<-0.0397 0.0814 -0.069 -0.9935>}
 pinky_01_l {parent=pinky_metacarpal_l, loc=<7.1709 -0.4831 -0.1094>, rot=<-0.0654 -0.0035 0.2242 -0.9723>}
 pinky_02_l {parent=pinky_01_l, loc=<2.8274 0.0893 -0.1043>, rot=<0.0037 0.0011 0.2854 -0.9584>}
 pinky_03_l {parent=pinky_02_l, loc=<2.2046 0.0307 -0.0078>, rot=<-0.0002 0.0098 0.1445 -0.9895>}
 ring_metacarpal_l {parent=hand_l, loc=<2.2074 -0.3625 1.3638>, rot=<-0.0284 0.0722 -0.0458 -0.9959>}
 ring_01_l {parent=ring_metacarpal_l, loc=<7.0295 -0.6197 -0.1446>, rot=<-0.0489 -0.0252 0.1892 -0.9804>}
 ring_02_l {parent=ring_01_l, loc=<3.6174 0.4647 -0.1027>, rot=<0 0 0.2806 -0.9598>}
 ring_03_l {parent=ring_02_l, loc=<2.4152 0.114 0.0199>, rot=<0 0 0.1189 -0.9929>}
 thumb_01_l {parent=hand_l, loc=<1.3118 -0.3191 -2.459>, rot=<0.6546 -0.4042 0.0664 -0.6354>}
 thumb_02_l {parent=thumb_01_l, loc=<5.9336 -0.0448 0.0401>, rot=<-0.0037 0.0213 0.298 -0.9543>}
 thumb_03_l {parent=thumb_02_l, loc=<2.2223 0.3682 -0.0907>, rot=<0.0118 -0.013 0.0066 -0.9998>}
 lowerarm_twist_01_l {parent=lowerarm_l, loc=<17.8549 0 0>, rot=<0 0 0 1>}
 lowerarm_twist_02_l {parent=lowerarm_l, loc=<9.6199 0 0>, rot=<0 0 0 1>}
 upperarm_twist_02_l {parent=upperarm_l, loc=<13.7941 0 0>, rot=<0 0 0 1>}
 dyn_shoulderArmor_Base_l {parent=clavicle_l, loc=<14.4688 -0.2863 6.9115>, rot=<-0.2485 -0.677 -0.6485 0.2436>}
 dyn_shoulderArmor_l {parent=dyn_shoulderArmor_Base_l, loc=<0 0 0>, rot=<0 0 0 1>}
 dyn_shoulderScale_l {parent=dyn_shoulderArmor_l, loc=<-8.0224 -7.6412 0>, rot=<0 0 0 1>}
 clavicle_r {parent=spine_05, loc=<4.3575 -0.1466 2.3223>, rot=<0.7289 0.0465 -0.6829 0.009>}
 upperarm_r {parent=clavicle_r, loc=<-15.5111 0 0>, rot=<0.0226 -0.3234 -0.0088 -0.9459>}
 upperarm_twist_01_r {parent=upperarm_r, loc=<-0.0041 0 0>, rot=<0 0 0 1>}
 lowerarm_r {parent=upperarm_r, loc=<-26.6515 0 0>, rot=<0.0391 -0.0099 -0.2453 -0.9686>}
 hand_r {parent=lowerarm_r, loc=<-22.3492 0 0>, rot=<-0.7081 -0.0507 0.0975 -0.6975>}
 index_metacarpal_r {parent=hand_r, loc=<-1.6534 -0.0181 2.9601>, rot=<0.0158 -0.0124 -0.0173 -0.9996>}
 index_01_r {parent=index_metacarpal_r, loc=<-8.3008 -0.0357 0.0524>, rot=<-0.0044 0.0485 0.0924 -0.9945>}
 index_02_r {parent=index_01_r, loc=<-3.3589 0.3003 -0.0453>, rot=<-0.002 0.0027 0.3225 -0.9466>}
 index_03_r {parent=index_02_r, loc=<-2.5578 0.1231 0.0251>, rot=<-0.002 0.0067 0.2558 -0.9667>}
 middle_metacarpal_r {parent=hand_r, loc=<-2.1856 0.1726 0.5984>, rot=<0.006 0.0168 -0.0329 -0.9993>}
 middle_01_r {parent=middle_metacarpal_r, loc=<-7.3273 0.2447 -0.0129>, rot=<-0.0208 0.0042 0.0858 -0.9961>}
 middle_02_r {parent=middle_01_r, loc=<-4.0142 -0.1411 -0.07>, rot=<-0.0217 0.0259 0.3512 -0.9357>}
 middle_03_r {parent=middle_02_r, loc=<-2.6074 -0.0534 -0.1312>, rot=<0.0083 0.0188 0.1708 -0.9851>}
 pinky_metacarpal_r {parent=hand_r, loc=<-1.7617 0.9753 -3.3814>, rot=<-0.0397 0.0814 -0.069 -0.9935>}
 pinky_01_r {parent=pinky_metacarpal_r, loc=<-7.1707 0.4832 0.1094>, rot=<-0.0654 -0.0035 0.2242 -0.9723>}
 pinky_02_r {parent=pinky_01_r, loc=<-2.8274 -0.0892 0.1043>, rot=<0.0037 0.0011 0.2854 -0.9584>}
 pinky_03_r {parent=pinky_02_r, loc=<-2.2045 -0.031 0.0078>, rot=<-0.0002 0.0098 0.1445 -0.9895>}
 ring_metacarpal_r {parent=hand_r, loc=<-2.2073 0.3627 -1.3638>, rot=<-0.0284 0.0722 -0.0458 -0.9959>}
 ring_01_r {parent=ring_metacarpal_r, loc=<-7.0297 0.6197 0.1446>, rot=<-0.0489 -0.0252 0.1892 -0.9804>}
 ring_02_r {parent=ring_01_r, loc=<-3.6171 -0.4648 0.1027>, rot=<0 0 0.2806 -0.9598>}
 ring_03_r {parent=ring_02_r, loc=<-2.4153 -0.1141 -0.0199>, rot=<0 0 0.1189 -0.9929>}
 thumb_01_r {parent=hand_r, loc=<-1.3119 0.3192 2.459>, rot=<0.6546 -0.4042 0.0664 -0.6354>}
 thumb_02_r {parent=thumb_01_r, loc=<-5.9335 0.0447 -0.0401>, rot=<-0.0037 0.0213 0.298 -0.9543>}
 thumb_03_r {parent=thumb_02_r, loc=<-2.2223 -0.3682 0.0907>, rot=<0.0118 -0.013 0.0066 -0.9998>}
 lowerarm_twist_01_r {parent=lowerarm_r, loc=<-17.855 0 0>, rot=<0 0 0 1>}
 lowerarm_twist_02_r {parent=lowerarm_r, loc=<-9.6199 0 0>, rot=<0 0 0 1>}
 upperarm_twist_02_r {parent=upperarm_r, loc=<-13.7942 0 0>, rot=<0 0 0 1>}
 dyn_shoulderArmor_Base_r {parent=clavicle_r, loc=<-14.4687 0.2863 -6.9114>, rot=<-0.2485 -0.677 -0.6485 0.2436>}
 dyn_shoulderArmor_r {parent=dyn_shoulderArmor_Base_r, loc=<0 0 0>, rot=<0 0 0 1>}
 dyn_shoulderScale_r {parent=dyn_shoulderArmor_r, loc=<8.0222 7.6411 0>, rot=<0 0 0 1>}
 neck_01 {parent=spine_05, loc=<11.3138 0.1529 0>, rot=<0 0 -0.2003 -0.9797>}
 neck_02 {parent=neck_01, loc=<4.5627 0 0>, rot=<0 0 0.0504 -0.9987>}
 head {parent=neck_02, loc=<5.3914 0 0>, rot=<0 0 0.115 -0.9934>}
 dyn_pistol_Base {parent=spine_01, loc=<4.2731 -0.6174 14.82>, rot=<0.1001 -0.7322 0.6701 0.0697>}
 dyn_pistol {parent=dyn_pistol_Base, loc=<0 0 0>, rot=<0 0 0 1>}
 dyn_pouch_Base {parent=spine_01, loc=<6.6658 0.9952 -14.3195>, rot=<-0.0322 0.658 0.7522 0.0126>}
 dyn_pouch {parent=dyn_pouch_Base, loc=<0 0 0>, rot=<0 0 0 1>}
 dyn_frontLegSplit {parent=pelvis, loc=<0 0 0>, rot=<0 0 0 1>}
 thigh_l {parent=pelvis, loc=<-2.5693 -0.0109 -7.7501>, rot=<0.0012 0.0572 0.0155 -0.9982>}
 calf_l {parent=thigh_l, loc=<-39.0484 0 0>, rot=<0.0006 -0.0161 -0.0391 -0.9991>}
 calf_twist_01_l {parent=calf_l, loc=<-35.813 -0.0967 0.0397>, rot=<0.0237 0.0095 -0.0035 -0.9997>}
 calf_twist_02_l {parent=calf_l, loc=<-18.8064 -0.0967 0.0397>, rot=<0.0237 0.0095 -0.0035 -0.9997>}
 foot_l {parent=calf_l, loc=<-42.561 0.1077 0.4292>, rot=<0.0418 -0.0455 0.0343 -0.9975>}
 ball_l {parent=foot_l, loc=<-6.1599 13.8264 0.6806>, rot=<0.0001 -0.0001 -0.7063 -0.7079>}
 thigh_twist_01_l {parent=thigh_l, loc=<-21.6303 0 0>, rot=<0.0013 0 0 -1>}
 thigh_r {parent=pelvis, loc=<-2.5691 -0.0109 7.7501>, rot=<0.0572 -0.0012 0.9982 0.0155>}
 calf_r {parent=thigh_r, loc=<39.0487 0 0>, rot=<0.0006 -0.0161 -0.0391 -0.9991>}
 calf_twist_01_r {parent=calf_r, loc=<35.8132 0.0967 -0.0397>, rot=<0.0237 0.0095 -0.0035 -0.9997>}
 calf_twist_02_r {parent=calf_r, loc=<18.8064 0.0967 -0.0397>, rot=<0.0237 0.0095 -0.0035 -0.9997>}
 foot_r {parent=calf_r, loc=<42.5612 -0.1077 -0.4292>, rot=<0.0418 -0.0455 0.0343 -0.9975>}
 ball_r {parent=foot_r, loc=<6.1598 -13.8264 -0.6806>, rot=<0.0001 -0.0001 -0.7063 -0.7079>}
 thigh_twist_01_r {parent=thigh_r, loc=<21.6303 0 0>, rot=<0.0013 0 0 -1>}
 dyn_skirtBackMid_Base {parent=pelvis, loc=<-6.25 -14.4049 0.1383>, rot=<0.0298 -0.9996 0.0002 -0.0027>}
 dyn_skirtBackMid {parent=dyn_skirtBackMid_Base, loc=<0 0 0>, rot=<0 0 0 1>}
 dyn_skirtFrontMid_Base {parent=pelvis, loc=<-4.567 10.601 0>, rot=<0 0 0.999 0.0445>}
 dyn_skirtFrontMid {parent=dyn_skirtFrontMid_Base, loc=<0 0 0>, rot=<0 0 0 1>}
 dyn_kneeAim_l {loc=<12.2078 -0.9999 51.7592>, rot=<0.0315 0.7349 -0.027 -0.6769>}
 SPR_thigh_l {type=Curve, loc=<0 0 0>, rot=<0 0 0 1>}
 SPR_thigh_r {type=Curve, loc=<0 0 0>, rot=<0 0 0 1>}
]
variables [
 dateGliderDeployed {type=double, default=0}
 isCelebrating {type=double, default=0}
 isCrouching {type=double, default=0}
 isFrontEnd {type=bool, default=False}
 isPosing {type=double, default=0}
 isSkydiving {type=double, default=0}
]
