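# Physics asset for the skirt on the medium male HighTower Date body, named for use during emotes.
# Front and back mid skirt bones carry a capsule and a box, each constrained with narrow swing and twist limits to a tiny base box.
# Both bases and both thigh capsules link to the pelvis with free angular motion, so the asset covers only skirt, pelvis and thighs.
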PhysicsAsset /BRCosmetics/Characters/Player/Male/Medium/Bodies/M_MED_HighTower_Date/Meshes/M_MED_HighTower_Date_EmoteSkirtPhysics
bodies {
 dyn_skirtBackMid [{shapeType=Box, extentX=13.6251, extentY=2.8215, extentZ=19.0098, center=<14.7981 0.307 0>}]
 dyn_skirtBackMid_Base [{shapeType=Box, extentX=1.01, extentY=1.01, extentZ=1.01, rotation=<rot 0 -90 0>}]
 dyn_skirtFrontMid [{shapeType=Capsule, length=31.1464, radius=6.6896, center=<9.192 -6.3239 -0.0004>, rotation=<rot 90 -12.9074 -19.4664>}]
 dyn_skirtFrontMid_Base [{shapeType=Box, extentX=1.01, extentY=1.01, extentZ=1.01, rotation=<rot 0 -90 0>}]
 pelvis [{shapeType=Capsule, length=4.5788, radius=5.2249}]
 thigh_l [{shapeType=Capsule, length=23.8531, radius=7.4396, center=<-5.952 0.309 8.4629>}]
 thigh_r [{shapeType=Capsule, length=27.3449, radius=7.2525, center=<15.3344 -0.786 -9.1923>}]
}
constraints [
 {
  bone1Name=dyn_skirtBackMid_Base
  bone2Name=pelvis
  swing1LimitDegrees=45
  swing1Motion=Free
  swing2LimitDegrees=45
  swing2Motion=Free
  twistLimitDegrees=45
  twistMotion=Free
 }
 {bone1Name=dyn_skirtBackMid, bone2Name=dyn_skirtBackMid_Base, swing1LimitDegrees=20, swing2LimitDegrees=20, twistLimitDegrees=8}
 {
  bone1Name=dyn_skirtFrontMid_Base
  bone2Name=pelvis
  swing1LimitDegrees=45
  swing1Motion=Free
  swing2LimitDegrees=45
  swing2Motion=Free
  twistLimitDegrees=45
  twistMotion=Free
 }
 {bone1Name=dyn_skirtFrontMid, bone2Name=dyn_skirtFrontMid_Base, swing1LimitDegrees=20, swing2LimitDegrees=20, twistLimitDegrees=8}
 {
  bone1Name=thigh_l
  bone2Name=pelvis
  swing1LimitDegrees=45
  swing1Motion=Free
  swing2LimitDegrees=45
  swing2Motion=Free
  twistLimitDegrees=45
  twistMotion=Free
 }
 {
  bone1Name=thigh_r
  bone2Name=pelvis
  swing1LimitDegrees=45
  swing1Motion=Free
  swing2LimitDegrees=45
  swing2Motion=Free
  twistLimitDegrees=45
  twistMotion=Free
 }
]
