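# Physics asset for the M_MED_HighTower_Date male medium player body cosmetic, covering its dangling pistol and pouch accessories and the thighs.
# Box bodies on dyn_pistol and dyn_pouch each hinge to a small _Base box through tight, asymmetric swing and twist limits.
# Capsules cover both thighs, with an oversized offset capsule on the right thigh twist bone constrained to thigh_r.
# No other body parts have collision.
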PhysicsAsset /BRCosmetics/Characters/Player/Male/Medium/Bodies/M_MED_HighTower_Date/Meshes/M_MED_HighTower_Date_Physics
bodies {
 dyn_pistol [{shapeType=Box, extentX=6.331, extentY=13.1771, extentZ=27.2781, center=<8.5924 -1.3831 0.1074>, rotation=<rot 0 91.7562 89.9995>}]
 dyn_pistol_Base [{shapeType=Box, extentX=1.01, extentY=1.01, extentZ=1.01}]
 dyn_pouch [{shapeType=Box, extentX=4.2442, extentY=8.4521, extentZ=12.6727, center=<2.2541 -2.2777 -0.0921>, rotation=<rot 0.3358 -87.9525 -87.2036>}]
 dyn_pouch_Base [{shapeType=Box, extentX=1.01, extentY=1.01, extentZ=1.01}]
 thigh_l [{shapeType=Capsule, length=27.0895, radius=7.8399, center=<-6.4128 -0.0002 -1.0829>, rotation=<rot 84.989 -180 -180>}]
 thigh_r [{shapeType=Capsule, length=26.6381, radius=8.1049, rotation=<rot 90 0 0>}]
 thigh_twist_01_r [{shapeType=Capsule, length=28.5246, radius=15.5446, center=<-13.461 1.0565 -6.9507>, rotation=<rot 87.0055 -180 -180>}]
}
constraints [
 {bone1Name=thigh_twist_01_r, bone2Name=thigh_r, swing1LimitDegrees=45, swing2LimitDegrees=45, twistLimitDegrees=45}
 {bone1Name=dyn_pistol, bone2Name=dyn_pistol_Base, swing1LimitDegrees=8, swing2LimitDegrees=25, twistLimitDegrees=8}
 {bone1Name=dyn_pouch, bone2Name=dyn_pouch_Base, swing1LimitDegrees=8, swing2LimitDegrees=20, twistLimitDegrees=6}
]
