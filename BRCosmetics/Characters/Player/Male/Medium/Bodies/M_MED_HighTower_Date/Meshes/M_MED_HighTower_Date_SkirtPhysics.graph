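# Physics asset for the skirt of the medium male HighTower Date player body, covering the front and back skirt bones plus the pelvis and upper legs.
# Each skirt panel is a box body constrained with tight swing and twist limits to a tiny base box, which joins the pelvis with free rotation.
# Capsules on the pelvis, thighs and thigh twist bones form the leg volumes, chained to the pelvis by wider angular limits.
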PhysicsAsset /BRCosmetics/Characters/Player/Male/Medium/Bodies/M_MED_HighTower_Date/Meshes/M_MED_HighTower_Date_SkirtPhysics
bodies {
 dyn_skirtBackMid [{shapeType=Box, extentX=13.6251, extentY=2.8215, extentZ=19.0098, center=<14.7981 0.307 0>}]
 dyn_skirtBackMid_Base [{shapeType=Box, extentX=1.01, extentY=1.01, extentZ=1.01, rotation=<rot 0 -90 0>}]
 dyn_skirtFrontMid [{shapeType=Box, extentX=3.0278, extentY=34.1217, extentZ=15.8395, center=<11.581 -1.6045 0.1056>, rotation=<rot -0.0014 -87.9219 -89.9465>}]
 dyn_skirtFrontMid_Base [{shapeType=Box, extentX=1.01, extentY=1.01, extentZ=1.01, rotation=<rot 0 -90 0>}]
 pelvis [{shapeType=Capsule, length=4.5788, radius=5.2249}]
 thigh_l [{shapeType=Capsule, length=26.7163, radius=8.193, center=<-53.5932 -3.2249 0.102>, rotation=<rot 0 -86.1127 -88.1535>}]
 thigh_r [{shapeType=Capsule, length=26.6158, radius=8.1065, center=<53.5279 3.2361 -0.1509>, rotation=<rot 0 -85.8851 -88.0298>}]
 thigh_twist_01_l [{shapeType=Capsule, length=27.1644, radius=9.4209, center=<4.9992 -1.1474 -0.0073>, rotation=<rot 90 0 0>}]
 thigh_twist_01_r [{shapeType=Capsule, length=27.2333, radius=9.474, center=<-5 1.1475 0>, rotation=<rot 90 0 0>}]
}
constraints [
 {bone1Name=thigh_l, bone2Name=pelvis, swing1LimitDegrees=45, swing2LimitDegrees=45, twistLimitDegrees=45}
 {bone1Name=thigh_twist_01_l, bone2Name=thigh_l, swing1LimitDegrees=45, swing2LimitDegrees=45, twistLimitDegrees=45}
 {bone1Name=thigh_r, bone2Name=pelvis, swing1LimitDegrees=45, swing2LimitDegrees=45, twistLimitDegrees=45}
 {bone1Name=thigh_twist_01_r, bone2Name=thigh_r, swing1LimitDegrees=45, swing2LimitDegrees=45, twistLimitDegrees=45}
 {
  bone1Name=dyn_skirtBackMid_Base
  bone2Name=pelvis
  swing1LimitDegrees=45
  swing1Motion=Free
  swing2LimitDegrees=45
  swing2Motion=Free
  twistLimitDegrees=45
  twistMotion=Free
 }
 {bone1Name=dyn_skirtBackMid, bone2Name=dyn_skirtBackMid_Base, swing1LimitDegrees=20, swing2LimitDegrees=20, twistLimitDegrees=8}
 {
  bone1Name=dyn_skirtFrontMid_Base
  bone2Name=pelvis
  swing1LimitDegrees=45
  swing1Motion=Free
  swing2LimitDegrees=45
  swing2Motion=Free
  twistLimitDegrees=45
  twistMotion=Free
 }
 {bone1Name=dyn_skirtFrontMid, bone2Name=dyn_skirtFrontMid_Base, swing1LimitDegrees=20, swing2LimitDegrees=20, twistLimitDegrees=8}
]
