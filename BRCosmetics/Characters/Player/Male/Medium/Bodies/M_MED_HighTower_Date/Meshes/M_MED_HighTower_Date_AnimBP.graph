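# Post-process style AnimBP for the HighTower Date body cosmetic, layering procedural motion and cloth-like physics onto the copied parent mesh pose.
# A state-driven Control Rig feeds body RigidBody physics, cached before chained emote-skirt and skirt simulations.
# A weighted thigh/pelvis constraint and bone offset then adjust the front skirt, while backpack and cape attach bones scale to zero.
# The event graph sets emote weights, a thigh-curve flag and a front-end matching-glider flag.
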AnimBlueprint /BRCosmetics/Characters/Player/Male/Medium/Bodies/M_MED_HighTower_Date/Meshes/M_MED_HighTower_Date_AnimBP
target_skeleton=&/M_MED_HighTower_Date_Skeleton
graphs [
 AnimGraph {nodes [
   n1 = "Constraint - dyn_skirtFrontMid (PPP)"!(ComponentPose, Alpha=1, ConstraintWeights_0=0.4, ConstraintWeights_1:float=0.4, ConstraintWeights_2:float=0.2) -> (Pose) {
    alphaBoolBlend {blendOption=Linear}
    alphaScaleBiasClamp {bInterpResult=true, interpSpeedDecreasing=1, interpSpeedIncreasing=1}
    boneToModify {boneName=dyn_skirtFrontMid}
    constraintSetup [{targetBone {boneName=thigh_r}}, {targetBone {boneName=thigh_l}}, {targetBone {boneName=pelvis}}]
    constraintWeights [1,1,1]
   }
   Blend Poses by bool(BlendPose_1=Component To Local(ComponentPose=@n1), bActiveValue=Get SPR_thigh_l_bool() -> (SPR_thigh_l_bool:bool)) {blendType=Linear}
   n2 = RigidBody(ComponentPose=Local To Component(LocalPose=Use cached pose 'preSkirt'())) {
    bClampLinearTranslationLimitToRefPose=true
    bOverrideWorldGravity=true
    overridePhysicsAsset=&/M_MED_HighTower_Date_EmoteSkirtPhysics
    simulationSpace=BaseBoneSpace
    alphaBoolBlend {blendOption=Linear}
    alphaScaleBiasClamp {bInterpResult=true, interpSpeedDecreasing=20, interpSpeedIncreasing=20}
    baseBoneRef {boneName=root}
    componentAppliedLinearAccClamp=<3000 3000 3000>
    componentLinearAccScale=<0.35 0.35 0.45>
    componentLinearVelScale=<0.2 0.2 0.3>
    overrideWorldGravity {z=-980}
   }
   n3 = RigidBody(ComponentPose=@n2) {
    bClampLinearTranslationLimitToRefPose=true
    bOverrideWorldGravity=true
    overridePhysicsAsset=&/M_MED_HighTower_Date_SkirtPhysics
    simulationSpace=BaseBoneSpace
    alphaBoolBlend {blendOption=Linear}
    alphaScaleBiasClamp {bInterpResult=true, interpSpeedDecreasing=20, interpSpeedIncreasing=20}
    baseBoneRef {boneName=root}
    componentAppliedLinearAccClamp=<3000 3000 3000>
    componentLinearAccScale=<0.35 0.35 0.45>
    componentLinearVelScale=<0.2 0.2 0.3>
    overrideWorldGravity {z=-980}
   }
   n4 = "Constraint - dyn_skirtFrontMid (PPP)"(
    ComponentPose=@n3
    Alpha=Get ThroneFloatINV() -> (ThroneFloatINV:float)
    ConstraintWeights_0=0.4
    ConstraintWeights_1=0.4
    ConstraintWeights_2=0.2
   ) {
    alphaBoolBlend {blendOption=Linear}
    alphaScaleBiasClamp {bInterpResult=true, interpSpeedDecreasing=1, interpSpeedIncreasing=1}
    boneToModify {boneName=dyn_skirtFrontMid}
    constraintSetup [{targetBone {boneName=thigh_r}}, {targetBone {boneName=thigh_l}}, {targetBone {boneName=pelvis}}]
    constraintWeights [1,1,1]
   }
   n5 = "Transform (Modify) Bone - Bone: dyn_skirtFrontMid_Base"(ComponentPose=@n4, Alpha=1, Translation="0,-5,0", Rotation="0,0,0", Scale="1,1,1") {
    rotationMode=BMM_Additive
    rotationSpace=BCS_BoneSpace
    translationMode=BMM_Additive
    translationSpace=BCS_BoneSpace
    alphaBoolBlend {blendOption=Linear}
    alphaScaleBiasClamp {bInterpResult=true, interpSpeedDecreasing=1, interpSpeedIncreasing=1}
    boneToModify {boneName=dyn_skirtFrontMid_Base}
    translation {y=-5}
   }
   n6 = "Transform (Modify) Bone - Bone: attach_backpack"!(ComponentPose=@n5, bAlphaBoolEnabled:bool=True, Translation="0,0,0", Rotation="0,0,0", Scale="0,0,0") -> (Pose) {
    alphaInputType=Bool
    rotationMode=BMM_Additive
    rotationSpace=BCS_BoneSpace
    scaleMode=BMM_Replace
    scaleSpace=BCS_BoneSpace
    translationMode=BMM_Additive
    translationSpace=BCS_BoneSpace
    alphaBoolBlend {blendOption=Linear}
    alphaScaleBiasClamp {bInterpResult=true, interpSpeedDecreasing=1, interpSpeedIncreasing=1}
    boneToModify {boneName=attach_backpack}
    scale=<0 0 0>
   }
   n7 = "Transform (Modify) Bone - Bone: attach_cape"!(ComponentPose=@n6, bAlphaBoolEnabled:bool=True, Translation="0,0,0", Rotation="0,0,0", Scale="0,0,0") -> (Pose) {
    alphaInputType=Bool
    rotationMode=BMM_Additive
    rotationSpace=BCS_BoneSpace
    scaleMode=BMM_Replace
    scaleSpace=BCS_BoneSpace
    translationSpace=BCS_BoneSpace
    alphaBoolBlend {blendOption=Linear}
    alphaScaleBiasClamp {bInterpResult=true, interpSpeedDecreasing=1, interpSpeedIncreasing=1}
    boneToModify {boneName=attach_cape}
    scale=<0 0 0>
   }
   Output Pose(Result=Component To Local(ComponentPose=@n7)) {layerGroup=None}
   n8 = Control Rig!(
    Source=Copy Pose From Mesh() {bUseAttachedParent=true}
    Alpha
    isSkydiving:double=Get SkydiveFloat() -> (SkydiveFloat:float)
    isCrouching:double=Get CrouchFloat() -> (CrouchFloat:float)
    isPosing:double=Get FrontEndPosingFloat() -> (FrontEndPosingFloat:float)
    dateGliderDeployed:double=Get DateGliderDeployFloat() -> (DateGliderDeployFloat:float)
    isCelebrating:double=Get JubileeFloat() -> (JubileeFloat:float)
    isFrontEnd:bool=Get bIsInFrontEnd()
   ) -> (Pose) {
    alphaBoolBlend {blendOption=Linear}
    controlRigAssetReference {blueprintRigClass=&/M_MED_HighTower_Date_Body_CtrlRig.M_MED_HighTower_Date_Body_CtrlRig_C}
   }
   n9 = RigidBody(ComponentPose=Local To Component(LocalPose=@n8)) {
    bClampLinearTranslationLimitToRefPose=true
    bOverrideWorldGravity=true
    overridePhysicsAsset=&/M_MED_HighTower_Date_Physics
    simulationSpace=BaseBoneSpace
    alphaBoolBlend {blendOption=Linear}
    baseBoneRef {boneName=root}
    componentAppliedLinearAccClamp=<3000 3000 3000>
    componentLinearAccScale=<0.35 0.35 0.45>
    componentLinearVelScale=<0.2 0.2 0.3>
    overrideWorldGravity {z=-980}
   }
   Save cached pose 'preSkirt'(Pose=Component To Local(ComponentPose=@n9)) {cacheName=preSkirt}
   Get skydiveFloat_b() -> (skydiveFloat_b:double)
   Get crouchFloat_b() -> (crouchFloat_b:double)
   Get dateGliderDeploy_b() -> (dateGliderDeploy_b:double)
   Get frontEndPosing_b() -> (frontEndPosing_b:double)
   Get jubilee_b() -> (jubilee_b:double)
  ]}
 EventGraph {nodes [
   Event BlueprintUpdateAnimation()
   > n1 = Sequence!(execute) -> (then_0, then_1, then_2:Exec)
   > Branch(Condition=Get bIsPlayingEmote())
   > Delay(Duration=2)
   n2 = GetCurveValue(CurveName=SPR_thigh_l)
   > Set SPR_thigh_l_clamp(SPR_thigh_l_clamp:double=@n2) -> (Output_Get:double)
   > n3 = Branch(Condition="Equal (Float)"(A=@n2, B=1))
   > Set SPR_thigh_l_bool(SPR_thigh_l_bool:bool=true) -> (Output_Get:bool)
   Set SPR_thigh_l_bool(execute=@n3.else, SPR_thigh_l_bool=false)
   n4 = Branch(execute=@n1.then_1, Condition=Get bIsPlayingEmote())
   > Set playing emote(playing emote:double=0.2) -> (Output_Get:double)
   > Set non emote alpha(non emote alpha:double=0) -> (Output_Get:double)
   Set playing emote(execute=@n4.else, playing emote=0)
   > Set non emote alpha(non emote alpha=1)
   n5 = Cast To FortPlayerPawn(execute=@n1.then_2, Object=GetOwningActor())
   n6 = Get ParachuteAttachment!(self:Fort Player Pawn Object Reference=@n5) -> (ParachuteAttachment:Fort Player Parachute Object Reference)
   > Is Valid(InputObject=@n6)
   n7 = Get ParachuteMesh!(self:Fort Player Parachute Object Reference=@n6) -> (ParachuteMesh:Skeletal Mesh Component Budgeted Object Reference)
   > Is Valid(InputObject=@n7)
   n8 = Get SkeletalMesh!(self:Skinned Mesh Component Object Reference=@n7) -> (SkeletalMesh:Skeletal Mesh Object Reference)
   > Is Valid(InputObject=@n8)
   > n9 = Branch(Condition=Get bIsInFrontEnd())
   > Set IsDateGlider(IsDateGlider:bool="Equal (SoftObjectReference)"(
     A:"Soft Object Reference (by ref)"=Make Soft Reference!(Input:Skeletal Mesh Object Reference=@n8) -> (Output:Skeletal Mesh Soft Object Reference)
     B:"Soft Object Reference (by ref)"=Get DateGlider() -> (DateGlider:Skeletal Mesh Soft Object Reference)
    ) -> (ReturnValue:bool=~) {is=title K2Node_PromotableOperator}) -> (Output_Get:bool)
   Set IsDateGlider(execute=@n9.else, IsDateGlider=false)
   n10 = Branch(Condition=Get bIsInFrontEnd())
   > Set FE Simspace Switch(FE Simspace Switch:AnimPhysSimSpaceType Enum=World) -> (Output_Get:AnimPhysSimSpaceType Enum)
   Set FE Simspace Switch(execute=@n10.else, FE Simspace Switch=BoneRelative)
  ]}
]
variables [
 crouchFloat_b {type=number}
 DateGlider {type=SkeletalMesh, default=/BRCosmetics/Gadgets/Assets/VinderTech_GliderChute/Glider_HighTower_Date/Meshes/Glider_HighTower_Date}
 dateGliderDeploy_b {type=number}
 FE Simspace Switch {type=AnimPhysSimSpaceType}
 frontEndPosing_b {type=number}
 HighTowerDate_Glider {type=SkeletalMesh, default=null}
 IsDateGlider {type=boolean}
 jubilee_b {type=number}
 non emote alpha {type=number}
 playing emote {type=number}
 skydiveFloat_b {type=number, default=1}
 SPR_thigh_l_bool {type=boolean}
 SPR_thigh_l_clamp {type=number}
 throne_b {type=boolean}
 throneFloat_b {type=number}
 throneFloatINV_b {type=number, default=1}
]
