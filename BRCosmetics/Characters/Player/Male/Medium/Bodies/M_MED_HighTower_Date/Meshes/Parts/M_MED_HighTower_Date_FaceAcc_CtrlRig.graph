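# Control Rig for the face accessory part of the HighTower Date medium male body, correcting hood bones as the head rotates.
# It converts the head's rotation relative to spine_05 to Euler angles and remaps them into clamped weights.
# Negative Z rotation blends in additive local translations on the three back hood bones.
# The two X-driven modifiers name no bone, and the Y remap is unconnected.
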ControlRigBlueprint /BRCosmetics/Characters/Player/Male/Medium/Bodies/M_MED_HighTower_Date/Meshes/Parts/M_MED_HighTower_Date_FaceAcc_CtrlRig
graphs [
 RigVMFunctionLibrary {links [], nodes []}
 RigVMModel {nodes [
   RigUnit_BeginExecution = &RigUnit_BeginExecution()
   n1 = &RigVMFunction_MathQuaternionToEuler(RotationOrder=ZYX, Value=&RigUnit_GetRelativeBoneTransform(Bone=head, Space=spine_05).Transform.Rotation)
   /Script/ControlRig.RigUnit_ModifyBoneTransforms(
    BoneToModify=((Transform=(Rotation=(X=0,Y=0,Z=0,W=1),Translation=(X=0,Y=0,Z=0),Scale3D=(X=1,Y=1,Z=1))))
    Mode=AdditiveLocal
    WeightMaximum=1
    Weight=&RigVMFunction_MathFloatRemap(SourceMaximum=-61, SourceMinimum=0, TargetMinimum=0, bClamp=true, Value=@n1.Result.X).Result
    ExecutePin=@RigUnit_BeginExecution.ExecutePin
   ) {is=struct RigVMUnitNode}
   > &RigUnit_ModifyBoneTransforms(
    BoneToModify=((Transform=(Rotation=(X=0,Y=0,Z=0,W=1),Translation=(X=0,Y=0,Z=0),Scale3D=(X=1,Y=1,Z=1))))
    Mode=AdditiveLocal
    WeightMaximum=1
    Weight=&RigVMFunction_MathFloatRemap(SourceMaximum=70, SourceMinimum=0, TargetMinimum=0, bClamp=true, Value=@n1.Result.X).Result
   )
   n2 = &RigVMFunction_MathFloatRemap(SourceMaximum=-48, SourceMinimum=-2, TargetMinimum=0, bClamp=true, Value=@n1.Result.Z)
   > &RigUnit_ModifyBoneTransforms(
    BoneToModify=((Bone="dyn_hoodBack_r",Transform=(Rotation=(X=0,Y=0,Z=0,W=1),Translation=(X=5,Y=2,Z=10),Scale3D=(X=1,Y=1,Z=1))))
    Mode=AdditiveLocal
    WeightMaximum=1
    WeightMinimum=0
    Weight=@n2.Result
   )
   > &RigUnit_ModifyBoneTransforms(
    BoneToModify=((Bone="dyn_hoodBack",Transform=(Rotation=(X=0,Y=0,Z=0,W=1),Translation=(X=10,Y=-2,Z=0),Scale3D=(X=1,Y=1,Z=1))))
    Mode=AdditiveLocal
    WeightMaximum=1
    Weight=@n2.Result
   )
   > &RigUnit_ModifyBoneTransforms(
    BoneToModify=((Bone="dyn_hoodBack_l",Transform=(Rotation=(X=0,Y=0,Z=0,W=1),Translation=(X=5,Y=2,Z=6),Scale3D=(X=1,Y=1,Z=1))))
    Mode=AdditiveLocal
    WeightMaximum=1
    Weight=@n2.Result
   )
   &RigVMFunction_MathFloatRemap(SourceMaximum=1, bClamp=true, Value=@n1.Result.Y)
  ]}
]
hierarchy [
 root {loc=<0 0 0>, rot=<0 0 0 1>}
 pelvis {parent=root, loc=<0 -0.7071 93.1191>, rot=<0.0127 0.707 -0.0127 -0.707>}
 spine_01 {parent=pelvis, loc=<6.7494 -0.6388 0>, rot=<0 0 0.034 -0.9994>}
 spine_02 {parent=spine_01, loc=<7.2518 -0.035 0>, rot=<0 0 0.0238 -0.9997>}
 spine_03 {parent=spine_02, loc=<6.0104 0 0>, rot=<0 0 0.0149 -0.9999>}
 spine_04 {parent=spine_03, loc=<8.1024 0 0>, rot=<0 0 -0.0132 -0.9999>}
 spine_05 {parent=spine_04, loc=<10.0146 0 0>, rot=<0 0 -0.0015 -1>}
 clavicle_l {parent=spine_05, loc=<4.3575 -0.1466 -2.3223>, rot=<0.0465 -0.7289 -0.009 -0.6829>}
 upperarm_l {parent=clavicle_l, loc=<15.5111 0 0>, rot=<0.0226 -0.3234 -0.0088 -0.9459>}
 upperarm_twist_01_l {parent=upperarm_l, loc=<0.0041 0 0>, rot=<0 0 0 -1>}
 clavicle_r {parent=spine_05, loc=<4.3575 -0.1466 2.3223>, rot=<0.7289 0.0465 -0.6829 0.009>}
 upperarm_r {parent=clavicle_r, loc=<-15.5111 0 0>, rot=<0.0226 -0.3234 -0.0088 -0.9459>}
 upperarm_twist_01_r {parent=upperarm_r, loc=<-0.0041 0 0>, rot=<0 0 0 1>}
 neck_01 {parent=spine_05, loc=<11.3138 0.1529 0>, rot=<0 0 -0.2003 -0.9797>}
 neck_02 {parent=neck_01, loc=<4.5627 0 0>, rot=<0 0 0.0504 -0.9987>}
 head {parent=neck_02, loc=<5.3914 0 0>, rot=<0 0 0.115 -0.9934>}
 dyn_hoodPoint {parent=head, loc=<13.5725 -3.1228 -0.4722>, rot=<0.6878 -0.7251 0.0234 0.0219>}
 dyn_hoodTop_Base {parent=head, loc=<11.174 10.072 0.2626>, rot=<0.0086 0.0047 0.8847 0.4661>}
 dyn_hoodTop {parent=dyn_hoodTop_Base, loc=<0 0 0>, rot=<0 0 0 -1>}
 dyn_hoodSide_Base_r {parent=head, loc=<-0.4725 1.5259 9.1634>, rot=<-0.4676 -0.5153 0.5927 0.4056>}
 dyn_hoodSide_r {parent=dyn_hoodSide_Base_r, loc=<0 0 0>, rot=<0 0 0 1>}
 dyn_hoodSide_Base_l {parent=head, loc=<-0.4729 1.5259 -9.1634>, rot=<-0.4056 -0.5927 0.5153 0.4676>}
 dyn_hoodSide_l {parent=dyn_hoodSide_Base_l, loc=<0 0 0>, rot=<0 0 0 1>}
 dyn_hoodBack_Base_r {parent=neck_02, loc=<3.6414 -6.0617 7.1117>, rot=<-0.6805 0.3247 -0.2595 -0.6035>}
 dyn_hoodBack_r {parent=dyn_hoodBack_Base_r, loc=<0 0 0>, rot=<0 0 0 -1>}
 dyn_hoodBack_Base {parent=neck_02, loc=<2.5535 -9.271 0.0312>, rot=<0.6455 -0.7634 0.0057 -0.0214>}
 dyn_hoodBack {parent=dyn_hoodBack_Base, loc=<0 0 0>, rot=<0 0 0 1>}
 dyn_hoodBack_Base_l {parent=neck_02, loc=<3.6416 -6.0617 -7.1117>, rot=<0.6589 -0.3337 -0.2481 -0.6268>}
 dyn_hoodBack_l {parent=dyn_hoodBack_Base_l, loc=<0 0 0>, rot=<0 0 0 1>}
]
