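# Physics asset for the hood of the male medium HighTower Date body, giving the dynamic hood bones simple collision shapes and joints.
# Head and neck capsules form the root chain; top and side hood chains hang from the head through base bodies, back chains from neck_02.
# Top, side and point hood joints have tight angular limits, while neck, base and back joints leave swing and twist free.
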PhysicsAsset /BRCosmetics/Characters/Player/Male/Medium/Bodies/M_MED_HighTower_Date/Meshes/Parts/M_MED_HighTower_Date_Physics
bodies {
 dyn_hoodBack [{shapeType=Box, extentX=2.5222, extentY=2.4091, extentZ=10.9821, center=<1.2684 0.8022 -0.6076>, rotation=<rot 2.9294 -90 7.5246>}]
 dyn_hoodBack_Base [{shapeType=Box, extentX=1.01, extentY=1.01, extentZ=1.01, rotation=<rot 0 -90 0>}]
 dyn_hoodBack_Base_l [{shapeType=Box, extentX=1.01, extentY=1.01, extentZ=1.01, rotation=<rot 0 -90 0>}]
 dyn_hoodBack_Base_r [{shapeType=Box, extentX=1.01, extentY=1.01, extentZ=1.01, rotation=<rot 0 -90 0>}]
 dyn_hoodBack_l [{shapeType=Box, extentX=1.01, extentY=1.01, extentZ=1.01, rotation=<rot 0 -90 0>}]
 dyn_hoodBack_r [{shapeType=Box, extentX=1.01, extentY=1.01, extentZ=1.01, rotation=<rot 0 -90 0>}]
 dyn_hoodPoint [{shapeType=Capsule, length=3.4277, radius=2.9718, center=<3.5312 0.9499 -0.2574>, rotation=<rot 1.1299 102.5191 -87.2076>}]
 dyn_hoodSide_Base_l [{shapeType=Box, extentX=1.01, extentY=1.01, extentZ=1.01, rotation=<rot 0 -90 0>}]
 dyn_hoodSide_Base_r [{shapeType=Box, extentX=1.01, extentY=1.01, extentZ=1.01, rotation=<rot 0 -90 0>}]
 dyn_hoodSide_l [{shapeType=Box, extentX=1.9556, extentY=9.6847, extentZ=8.811, center=<4.1349 -0.0004 0>, rotation=<rot -5.6115 89.9988 -180>}]
 dyn_hoodSide_r [{shapeType=Box, extentX=1.9556, extentY=9.6847, extentZ=8.811, center=<3.673 0.0001 0>, rotation=<rot 8.7861 89.9998 0>}]
 dyn_hoodTop [{shapeType=Box, extentX=2.5062, extentY=6.9302, extentZ=9.5558, center=<2.893 -1.6947 -0.1815>, rotation=<rot -1.0385 -89.9999 3.3907>}]
 dyn_hoodTop_Base [{shapeType=Box, extentX=1.01, extentY=1.01, extentZ=1.01, rotation=<rot 0 -90 0>}]
 head [{shapeType=Capsule, length=9.4771, radius=4.0709, center=<0.6463 1.6911 0>, rotation=<rot 90 0 -14.7504>}]
 neck_01 [{shapeType=Capsule, length=0.505, radius=0.505, rotation=<rot 0 -90 0>}]
 neck_02 [{shapeType=Capsule, length=0.505, radius=0.505, rotation=<rot 0 -90 0>}]
}
constraints [
 {
  bone1Name=neck_02
  bone2Name=neck_01
  swing1LimitDegrees=45
  swing1Motion=Free
  swing2LimitDegrees=45
  swing2Motion=Free
  twistLimitDegrees=45
  twistMotion=Free
 }
 {
  bone1Name=head
  bone2Name=neck_02
  swing1LimitDegrees=45
  swing1Motion=Free
  swing2LimitDegrees=45
  swing2Motion=Free
  twistLimitDegrees=45
  twistMotion=Free
 }
 {bone1Name=dyn_hoodPoint, bone2Name=head, swing1LimitDegrees=18, swing2LimitDegrees=12, twistLimitDegrees=12}
 {
  bone1Name=dyn_hoodTop_Base
  bone2Name=head
  swing1LimitDegrees=45
  swing1Motion=Free
  swing2LimitDegrees=45
  swing2Motion=Free
  twistLimitDegrees=45
  twistMotion=Free
 }
 {
  bone1Name=dyn_hoodSide_Base_r
  bone2Name=head
  swing1LimitDegrees=45
  swing1Motion=Free
  swing2LimitDegrees=45
  swing2Motion=Free
  twistLimitDegrees=45
  twistMotion=Free
 }
 {
  bone1Name=dyn_hoodSide_Base_l
  bone2Name=head
  swing1LimitDegrees=45
  swing1Motion=Free
  swing2LimitDegrees=45
  swing2Motion=Free
  twistLimitDegrees=45
  twistMotion=Free
 }
 {
  bone1Name=dyn_hoodBack_Base_r
  bone2Name=neck_02
  swing1LimitDegrees=45
  swing1Motion=Free
  swing2LimitDegrees=45
  swing2Motion=Free
  twistLimitDegrees=45
  twistMotion=Free
 }
 {
  bone1Name=dyn_hoodBack_Base
  bone2Name=neck_02
  swing1LimitDegrees=45
  swing1Motion=Free
  swing2LimitDegrees=45
  swing2Motion=Free
  twistLimitDegrees=45
  twistMotion=Free
 }
 {
  bone1Name=dyn_hoodBack_Base_l
  bone2Name=neck_02
  swing1LimitDegrees=45
  swing1Motion=Free
  swing2LimitDegrees=45
  swing2Motion=Free
  twistLimitDegrees=45
  twistMotion=Free
 }
 {bone1Name=dyn_hoodTop, bone2Name=dyn_hoodTop_Base, swing1LimitDegrees=25, swing2LimitDegrees=12, twistLimitDegrees=6}
 {bone1Name=dyn_hoodSide_r, bone2Name=dyn_hoodSide_Base_r, swing1LimitDegrees=25, swing2LimitDegrees=12, twistLimitDegrees=10}
 {
  bone1Name=dyn_hoodBack_r
  bone2Name=dyn_hoodBack_Base_r
  swing1LimitDegrees=45
  swing1Motion=Free
  swing2LimitDegrees=45
  swing2Motion=Free
  twistLimitDegrees=45
  twistMotion=Free
 }
 {
  bone1Name=dyn_hoodBack
  bone2Name=dyn_hoodBack_Base
  swing1LimitDegrees=45
  swing1Motion=Free
  swing2LimitDegrees=45
  swing2Motion=Free
  twistLimitDegrees=45
  twistMotion=Free
 }
 {
  bone1Name=dyn_hoodBack_l
  bone2Name=dyn_hoodBack_Base_l
  swing1LimitDegrees=45
  swing1Motion=Free
  swing2LimitDegrees=45
  swing2Motion=Free
  twistLimitDegrees=45
  twistMotion=Free
 }
 {bone1Name=dyn_hoodSide_l, bone2Name=dyn_hoodSide_Base_l, swing1LimitDegrees=25, swing2LimitDegrees=12, twistLimitDegrees=10}
]
